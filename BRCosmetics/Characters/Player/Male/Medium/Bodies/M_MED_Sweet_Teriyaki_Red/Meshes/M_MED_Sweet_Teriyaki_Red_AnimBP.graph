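# Post-process style AnimBP for the Sweet Teriyaki Red body mesh that corrects thigh volume bones.
# It copies the parent mesh pose, then blends six volume bones between thigh twist and pelvis with constraints (front-left adds spine_01).
# Additive bone-space offsets follow, scaled by alphas computed each update from thigh-to-pelvis reference transform deltas.
# The right-side thigh_volume_rt_r offset runs at constant full alpha, leaving its computed thigh_shape_rt_r unused.
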AnimBlueprint /BRCosmetics/Characters/Player/Male/Medium/Bodies/M_MED_Sweet_Teriyaki_Red/Meshes/M_MED_Sweet_Teriyaki_Red_AnimBP
target_skeleton=&/M_MED_Sweet_Teriyaki_Red_Skeleton
graphs [
 AnimGraph {nodes [
   n1 = "Constraint - thigh_volume_bk_l (PP)"!(
    ComponentPose=Local To Component(LocalPose=Copy Pose From Mesh() {bUseAttachedParent=true})
    Alpha=1
    ConstraintWeights_0=0.5
    ConstraintWeights_1:float=0.5
   ) -> (Pose) {
    alphaBoolBlend {blendOption=Linear}
    boneToModify {boneName=thigh_volume_bk_l}
    constraintSetup [{targetBone {boneName=thigh_twist_01_l}}, {targetBone {boneName=pelvis}}]
    constraintWeights [1,1]
   }
   n2 = "Constraint - thigh_volume_fn_l (PPP)"!(ComponentPose=@n1, Alpha=1, ConstraintWeights_0=0.45, ConstraintWeights_1:float=0.45, ConstraintWeights_2:float=0.1) -> (Pose) {
    alphaBoolBlend {blendOption=Linear}
    boneToModify {boneName=thigh_volume_fn_l}
    constraintSetup [{targetBone {boneName=thigh_twist_01_l}}, {targetBone {boneName=pelvis}}, {targetBone {boneName=spine_01}}]
    constraintWeights [1,1,1]
   }
   n3 = "Constraint - thigh_volume_lt_l1 (PP)"!(ComponentPose=@n2, Alpha=1, ConstraintWeights_0=0.5, ConstraintWeights_1:float=0.5) -> (Pose) {
    alphaBoolBlend {blendOption=Linear}
    boneToModify {boneName=thigh_volume_lt_l1}
    constraintSetup [{targetBone {boneName=thigh_twist_01_l}}, {targetBone {boneName=pelvis}}]
    constraintWeights [1,1]
   }
   n4 = "Constraint - thigh_volume_bk_r (PP)"!(ComponentPose=@n3, Alpha=1, ConstraintWeights_0=0.5, ConstraintWeights_1:float=0.5) -> (Pose) {
    alphaBoolBlend {blendOption=Linear}
    boneToModify {boneName=thigh_volume_bk_r}
    constraintSetup [{targetBone {boneName=thigh_twist_01_r}}, {targetBone {boneName=pelvis}}]
    constraintWeights [1,1]
   }
   n5 = "Constraint - thigh_volume_fn_r (PP)"!(ComponentPose=@n4, Alpha=1, ConstraintWeights_0=0.5, ConstraintWeights_1:float=0.5) -> (Pose) {
    alphaBoolBlend {blendOption=Linear}
    boneToModify {boneName=thigh_volume_fn_r}
    constraintSetup [{targetBone {boneName=thigh_twist_01_r}}, {targetBone {boneName=pelvis}}]
    constraintWeights [1,1]
   }
   n6 = "Constraint - thigh_volume_rt_r (PP)"!(ComponentPose=@n5, Alpha=1, ConstraintWeights_0=0.5, ConstraintWeights_1:float=0.5) -> (Pose) {
    alphaBoolBlend {blendOption=Linear}
    boneToModify {boneName=thigh_volume_rt_r}
    constraintSetup [{targetBone {boneName=thigh_twist_01_r}}, {targetBone {boneName=pelvis}}]
    constraintWeights [1,1]
   }
   n7 = "Transform (Modify) Bone - Bone: thigh_volume_fn_l"(ComponentPose=@n6, Alpha=Get thigh_shape_fn_l() -> (thigh_shape_fn_l:double), Translation="0,30,10", Rotation="0,0,-30", Scale="1,1,1") {
    rotationMode=BMM_Additive
    rotationSpace=BCS_BoneSpace
    scaleMode=BMM_Additive
    scaleSpace=BCS_BoneSpace
    translationMode=BMM_Additive
    translationSpace=BCS_BoneSpace
    alphaBoolBlend {blendOption=Linear}
    boneToModify {boneName=thigh_volume_fn_l}
    rotation {roll=-30}
    translation {y=30, z=10}
   }
   n8 = "Transform (Modify) Bone - Bone: thigh_volume_fn_r"(ComponentPose=@n7, Alpha=Get thigh_shape_fn_r() -> (thigh_shape_fn_r:double), Translation="0,-10,0", Rotation="0,0,0", Scale="1,1,1") {
    rotationSpace=BCS_BoneSpace
    scaleMode=BMM_Additive
    scaleSpace=BCS_BoneSpace
    translationMode=BMM_Additive
    translationSpace=BCS_BoneSpace
    alphaBoolBlend {blendOption=Linear}
    boneToModify {boneName=thigh_volume_fn_r}
    translation {y=-10}
   }
   n9 = "Transform (Modify) Bone - Bone: thigh_volume_bk_l"(ComponentPose=@n8, Alpha=Get thigh_shape_bk_l() -> (thigh_shape_bk_l:double), Translation="0,-13,0", Rotation="0,0,0", Scale="1,1,1") {
    rotationMode=BMM_Additive
    rotationSpace=BCS_BoneSpace
    scaleMode=BMM_Additive
    scaleSpace=BCS_BoneSpace
    translationMode=BMM_Additive
    translationSpace=BCS_BoneSpace
    alphaBoolBlend {blendOption=Linear}
    boneToModify {boneName=thigh_volume_bk_l}
    translation {y=-13}
   }
   n10 = "Transform (Modify) Bone - Bone: thigh_volume_bk_r"(ComponentPose=@n9, Alpha=Get thigh_shape_bk_r() -> (thigh_shape_bk_r:double), Translation="0,13,0", Rotation="0,0,0", Scale="1,1,1") {
    rotationSpace=BCS_BoneSpace
    scaleMode=BMM_Additive
    scaleSpace=BCS_BoneSpace
    translationMode=BMM_Additive
    translationSpace=BCS_BoneSpace
    alphaBoolBlend {blendOption=Linear}
    boneToModify {boneName=thigh_volume_bk_r}
    translation {y=13}
   }
   n11 = "Transform (Modify) Bone - Bone: thigh_volume_lt_l1"(ComponentPose=@n10, Alpha=Get thigh_shape_lt_l() -> (thigh_shape_lt_l:double), Translation="12,-5,0", Rotation="0,40,0", Scale="1,1,1") {
    rotationMode=BMM_Additive
    rotationSpace=BCS_BoneSpace
    scaleMode=BMM_Additive
    scaleSpace=BCS_BoneSpace
    translationMode=BMM_Additive
    translationSpace=BCS_BoneSpace
    alphaBoolBlend {blendOption=Linear}
    boneToModify {boneName=thigh_volume_lt_l1}
    rotation {yaw=40}
    translation=<12 -5>
   }
   n12 = "Transform (Modify) Bone - Bone: thigh_volume_rt_r"(ComponentPose=@n11, Alpha=1, Translation="0,5,0", Rotation="0,0,0", Scale="1,1,1") {
    rotationSpace=BCS_BoneSpace
    scaleMode=BMM_Additive
    scaleSpace=BCS_BoneSpace
    translationMode=BMM_Additive
    translationSpace=BCS_BoneSpace
    alphaBoolBlend {blendOption=Linear}
    boneToModify {boneName=thigh_volume_rt_r}
    translation {y=5}
   }
   Output Pose(Result=Component To Local(ComponentPose=@n12)) {layerGroup=None}
   Get thigh_shape_rt_r() -> (thigh_shape_rt_r:double)
  ]}
 EventGraph {nodes [
   Event BlueprintInitializeAnimation()
   > Set Skel_Mesh(Skel_Mesh=GetOwnerSkeletalMeshComponent())
   Event BlueprintUpdateAnimation()
   n1 = Get Skel_Mesh()
   > Set thigh_shape_fn_l(thigh_shape_fn_l:double=GetAlphaFromReferenceTransformDelta(SkeletalMeshComponent=@n1, BoneName=thigh_l, BaseName=pelvis, InputRange=(X=0,Y=-130), TwistAxis="0,0,1")) -> (Output_Get:double)
   > Set thigh_shape_fn_r(thigh_shape_fn_r:double=GetAlphaFromReferenceTransformDelta(SkeletalMeshComponent=@n1, BoneName=thigh_r, BaseName=pelvis, InputRange=(X=0,Y=-130), TwistAxis="0,0,1")) -> (Output_Get:double)
   n2 = Get Skel_Mesh()
   > Set thigh_shape_bk_l(thigh_shape_bk_l:double=GetAlphaFromReferenceTransformDelta(SkeletalMeshComponent=@n2, BoneName=thigh_l, BaseName=pelvis, InputRange=(X=0,Y=-110), TwistAxis="0,0,1")) -> (Output_Get:double)
   > Set thigh_shape_bk_r(thigh_shape_bk_r:double=GetAlphaFromReferenceTransformDelta(SkeletalMeshComponent=@n2, BoneName=thigh_r, BaseName=pelvis, InputRange=(X=0,Y=-110), TwistAxis="0,0,1")) -> (Output_Get:double)
   n3 = Get Skel_Mesh()
   > Set thigh_shape_lt_l(thigh_shape_lt_l:double=GetAlphaFromReferenceTransformDelta(SkeletalMeshComponent=@n3, BoneName=thigh_l, BaseName=pelvis, InputRange=(X=-150,Y=0), OutputRange=(X=1,Y=0), TwistAxis="0,0,1")) -> (Output_Get:double)
   > Set thigh_shape_rt_r(thigh_shape_rt_r:double=GetAlphaFromReferenceTransformDelta(SkeletalMeshComponent=@n3, BoneName=thigh_r, BaseName=pelvis, InputRange=(X=-150,Y=0), OutputRange=(X=1,Y=0), TwistAxis="0,0,1")) -> (Output_Get:double)
  ]}
]
variables [
 Skel_Mesh {type=SkeletalMeshComponent}
 thigh_shape_bk_l {type=number}
 thigh_shape_bk_r {type=number}
 thigh_shape_fn_l {type=number}
 thigh_shape_fn_r {type=number}
 thigh_shape_lt_l {type=number}
 thigh_shape_rt_r {type=number}
]
